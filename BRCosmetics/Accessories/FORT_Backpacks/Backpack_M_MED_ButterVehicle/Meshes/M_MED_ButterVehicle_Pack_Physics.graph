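# Physics asset for the ButterVehicle back bling, giving its skeleton a croissant base body and a chain of hair bodies.
# A large capsule covers the base, a zero-size master body anchors the hair, and five small capsules taper toward the tip.
# The master joint rotates freely against the base.
# Each hair link joins the one before it with tight twist and asymmetric swing limits, allowing more bend on one axis.
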PhysicsAsset /BRCosmetics/Accessories/FORT_Backpacks/Backpack_M_MED_ButterVehicle/Meshes/M_MED_ButterVehicle_Pack_Physics
bodies {
 dyn_Base_Croissant_jnt [{shapeType=Capsule, length=10, radius=8, shapeName=dyn_Base_Croissant_jnt_capsule, rotation=<rot 0 90 90>}]
 dyn_Hair_01_jnt [{shapeType=Capsule, length=1, radius=1.5, shapeName=dyn_Hair_01_jnt_capsule, center=<2 0 0.5917>, rotation=<rot 90 0 0>}]
 dyn_Hair_02_jnt [{shapeType=Capsule, length=1, radius=1.5, shapeName=dyn_Hair_02_jnt_capsule, center=<2 0 0>, rotation=<rot 90 0 0>}]
 dyn_Hair_03_jnt [{shapeType=Capsule, length=1, radius=1, shapeName=dyn_Hair_03_jnt_capsule, center=<2 0 0>, rotation=<rot 90 0 0>}]
 dyn_Hair_04_jnt [{shapeType=Capsule, length=1, radius=0.75, shapeName=dyn_Hair_04_jnt_capsule, center=<2 0 0>, rotation=<rot 90 0 0>}]
 dyn_Hair_05_jnt [{shapeType=Capsule, length=1, radius=0.5, shapeName=dyn_Hair_05_jnt_capsule, center=<2 0 0>, rotation=<rot 90 0 0>}]
 dyn_Hair_MSTR [{shapeType=Capsule, length=0, radius=0, shapeName=dyn_Hair_MSTR_capsule}]
}
constraints [
 {
  bone1Name=dyn_Hair_MSTR
  bone2Name=dyn_Base_Croissant_jnt
  swing1LimitDegrees=45
  swing1Motion=Free
  swing2LimitDegrees=45
  swing2Motion=Free
  twistLimitDegrees=45
  twistMotion=Free
 }
 {bone1Name=dyn_Hair_01_jnt, bone2Name=dyn_Hair_MSTR, swing1LimitDegrees=2.5, swing2LimitDegrees=15, twistLimitDegrees=1}
 {bone1Name=dyn_Hair_02_jnt, bone2Name=dyn_Hair_01_jnt, swing1LimitDegrees=2.5, swing2LimitDegrees=15, twistLimitDegrees=1}
 {bone1Name=dyn_Hair_03_jnt, bone2Name=dyn_Hair_02_jnt, swing1LimitDegrees=2.5, swing2LimitDegrees=15, twistLimitDegrees=1}
 {bone1Name=dyn_Hair_04_jnt, bone2Name=dyn_Hair_03_jnt, swing1LimitDegrees=2.5, swing2LimitDegrees=15, twistLimitDegrees=1}
 {bone1Name=dyn_Hair_05_jnt, bone2Name=dyn_Hair_04_jnt, swing1LimitDegrees=2.5, swing2LimitDegrees=15, twistLimitDegrees=1}
]
