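# Animation Blueprint for the ButterVehicle backpack cosmetic, adding secondary motion to its croissant base, ends, mouth and eyes.
# Six chained spring controllers feed a RigidBody simulation in pack-bone space, then a small additive root offset.
# Each update sets three small spring alphas that drop to zero when pawn speed exceeds a cutoff.
# It also feeds the simulation's gravity override from a deprecated gravity state function.
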AnimBlueprint /BRCosmetics/Accessories/FORT_Backpacks/Backpack_M_MED_ButterVehicle/Meshes/M_MED_ButterVehicle_Pack_AnimBP
target_skeleton=&/M_MED_ButterVehicle_Pack_Skeleton
graphs [
 AnimGraph {nodes [
   n1 = Get Spring Alpha 03() -> (Spring Alpha 03:double)
   n2 = Spring controller - Bone: dyn_Base_Croissant_jnt!(ComponentPose:Component Space Pose Link Structure=Local To Component(LocalPose=Input Pose()), Alpha:float=@n1) -> (Pose:Component Space Pose Link Structure) {
    is=title AnimGraphNode_SpringBone
    springDamping=5
    springStiffness=250
    alphaBoolBlend {blendOption=Linear}
    springBone {boneName=dyn_Base_Croissant_jnt}
   }
   n3 = Get SpringAlpha() -> (SpringAlpha:double)
   n4 = Spring controller - Bone: dyn_rt_Croissant_End_jnt!(ComponentPose:Component Space Pose Link Structure=@n2, Alpha:float=@n3) -> (Pose:Component Space Pose Link Structure) {
    is=title AnimGraphNode_SpringBone
    springDamping=5
    springStiffness=200
    alphaBoolBlend {blendOption=Linear}
    springBone {boneName=dyn_rt_Croissant_End_jnt}
   }
   n5 = Spring controller - Bone: dyn_lf_Croissant_End_jnt!(ComponentPose:Component Space Pose Link Structure=@n4, Alpha:float=@n3) -> (Pose:Component Space Pose Link Structure) {
    is=title AnimGraphNode_SpringBone
    springDamping=5
    springStiffness=250
    alphaBoolBlend {blendOption=Linear}
    springBone {boneName=dyn_lf_Croissant_End_jnt}
   }
   n6 = Spring controller - Bone: dyn_ct_Croissant_Mouth_jnt!(ComponentPose:Component Space Pose Link Structure=@n5, Alpha:float=Get SpringAlpha02() -> (SpringAlpha02:double)) -> (Pose:Component Space Pose Link Structure) {
    is=title AnimGraphNode_SpringBone
    springDamping=1.5
    springStiffness=225
    alphaBoolBlend {blendOption=Linear}
    springBone {boneName=dyn_ct_Croissant_Mouth_jnt}
   }
   n7 = Spring controller - Bone: dyn_lf_Croissant_Eye_jnt!(ComponentPose:Component Space Pose Link Structure=@n6, Alpha:float=@n1) -> (Pose:Component Space Pose Link Structure) {
    is=title AnimGraphNode_SpringBone
    springDamping=1.5
    springStiffness=250
    alphaBoolBlend {blendOption=Linear}
    springBone {boneName=dyn_lf_Croissant_Eye_jnt}
   }
   n8 = Spring controller - Bone: dyn_rt_Croissant_Eye_jnt!(ComponentPose:Component Space Pose Link Structure=@n7, Alpha:float=@n1) -> (Pose:Component Space Pose Link Structure) {
    is=title AnimGraphNode_SpringBone
    springDamping=1.5
    springStiffness=200
    alphaBoolBlend {blendOption=Linear}
    springBone {boneName=dyn_rt_Croissant_Eye_jnt}
   }
   n9 = RigidBody!(ComponentPose=@n8, Alpha, OverrideWorldGravity:Vector=Get GravityOverride(), ExternalForce) -> (Pose) {
    bClampLinearTranslationLimitToRefPose=true
    bOverrideWorldGravity=true
    overridePhysicsAsset=&/M_MED_ButterVehicle_Pack_Physics
    simulationSpace=BaseBoneSpace
    alphaBoolBlend {blendOption=Linear}
    baseBoneRef {boneName=pack}
    componentAppliedLinearAccClamp=<3000 3000 3000>
    componentLinearAccScale=<0.4 0.3 0.3>
    componentLinearVelScale=<0.4 0.3 0.3>
    simSpaceSettings {
     maxAngularAcceleration=1000
     maxAngularVelocity=3
     maxLinearAcceleration=700
     maxLinearVelocity=500
     velocityScaleZ=0
     worldAlpha=1
     externalLinearDragV=<0.2 0.2>
    }
   }
   n10 = "Transform (Modify) Bone - Bone: root"(ComponentPose=@n9, Translation="0,-1,5.5", Rotation="0,0,0") {
    translationMode=BMM_Additive
    translationSpace=BCS_BoneSpace
    alphaBoolBlend {blendOption=Linear}
    boneToModify {boneName=root}
    translation {y=-1, z=5.5}
   }
   Output Pose(Result=Component To Local(ComponentPose=@n10)) {layerGroup=None}
  ]}
 EventGraph {nodes [
   Event BlueprintUpdateAnimation()
   > n1 = Set pawnVelocitySpeed(pawnVelocitySpeed:double=Vector Length(A=Get PawnVelocity())) -> (Output_Get:double)
   > n2 = Set pawnVelocitySpeedCutoff(pawnVelocitySpeedCutoff:double=SelectFloat(A=0, B=1, bPickA="float > float"(A=@n1, B=1200) {is=title K2Node_CallFunction})) -> (Output_Get:double)
   > Set SpringAlpha(SpringAlpha:double=float * float(A=0.01, B=@n2) {is=title K2Node_PromotableOperator}) -> (Output_Get:double)
   > Set SpringAlpha02(SpringAlpha02:double=float * float(A=@n2, B=0.003)) -> (Output_Get:double)
   > Set Spring Alpha 03(Spring Alpha 03:double=float * float(A=@n2, B=0.005)) -> (Output_Get:double)
   > n3 = Gravity Override State Function DEPRECATED(CharacterPartAnimInstance=Self-Reference(), GameTimeInSeconds=GetGameTimeInSeconds(), GravityOverrideParams=Get Gravity Override Params())
   > Set GravityOverride(GravityOverride=@n3)
   TryGetPawnOwner()
  ]}
]
variables [
 Gravity Override Params {type=GravityOverrideParamsStruct}
 GravityOverride {type=Vector}
 pawnVelocitySpeed {type=number}
 pawnVelocitySpeedCutoff {type=number}
 Spring Alpha 03 {type=number}
 SpringAlpha {type=number}
 SpringAlpha02 {type=number}
]
